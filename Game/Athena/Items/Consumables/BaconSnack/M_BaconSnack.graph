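# Default-lit material for the BaconSnack consumable, built on beacon grenade bot textures, with hue-shifted base color, tinted blue-channel-masked emissive and a vertex-animated deploy.
# Vertex colors mask the animation: RGB selects legs and their rotation axes about a shared pivot, and inverted alpha pulls vertices toward the object pivot.
# A timestamp-driven smoothstep fades both offsets out over the deploy length.
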Material /Game/Athena/Items/Consumables/BaconSnack/M_BaconSnack
graph {
 properties {bUsedWithMeshParticles=true, bUsedWithNanite=true, pixelDepthOffsetMode=PDOM_Legacy, usedShadingModels=Default Lit}
 nodes [
  n1 = TextureSample() {texture=/Game/Creative/Sets/PropSets/Playgrounds/Textures/T_BL_Bot_Beacon_Grenade_E}
  n2 = VectorParameter() {parameterName=EmissiveColor, defaultValue=<rgba 0.5208 8.3399 20 0>}
  MaterialFunctionCall(Emissive=Multiply(A=Multiply(A=@n1.B, B=@n2.RGB), B=ScalarParameter() {defaultValue=1, parameterName=Emissive})) {materialFunction=&MF_FX_EmissiveAdjustments}
  MaterialFunctionCall(
   Hue Shift Percentage=ScalarParameter() {parameterName=Hue, sliderMax=1}
   Texture=TextureSample() {texture=/Game/Creative/Sets/PropSets/Playgrounds/Textures/T_BL_Bot_Beacon_Grenade_D}.RGB
  ) {materialFunction=&HueShift}
  n3 = OneMinus(None=VertexColor().A)
  n4 = Subtract(
   A=MaterialFunctionCall() {materialFunction=&ObjectPivotPoint}.Object Pivot Location
   B=WorldPosition().XYZ
  )
  n5 = DotProduct(A=VertexColor(), B=Constant() {r=1})
  n6 = VertexColor()
  n7 = LinearInterpolate(
   A=LinearInterpolate(
    A=VectorParameter() {parameterName=RotAxisRed, defaultValue=<rgba 1 -0.8 0 0>}.RGB
    B=VectorParameter() {parameterName=RotAxisGreen, defaultValue=<rgba -1 -0.9 0 0>}.RGB
    Alpha=@n6.G
   )
   B=VectorParameter() {parameterName=RotAxisBlue, defaultValue=<rgba 0 1 0 0>}.RGB
   Alpha=@n6.B
  )
  n8 = RotateAboutAxis(
   NormalizedRotationAxis=Normalize(VectorInput=Transform(None=@n7) {transformSourceType=TRANSFORMSOURCE_Local})
   RotationAngle=ScalarParameter() {defaultValue=-0.0625, parameterName=LegRotAngle}
   PivotPoint=TransformPosition(None=VectorParameter() {parameterName=PivotRed, defaultValue=<rgba 0 0 27.4 0>}.RGB) {transformType=TRANSFORMPOSSOURCE_World}
   Position=WorldPosition().XYZ
  )
  n9 = ScalarParameter() {parameterName=DeployAnimTimestamp}
  n10 = ScalarParameter() {defaultValue=0.5, parameterName=DeployAnimLength}
  Multiply(
   A=Add(A=Multiply(A=@n3, B=@n4), B=Multiply(A=@n5, B=@n8))
   B=OneMinus(None=MaterialFunctionCall(TimestampA=@n9, TransitionLength_In=@n10) {materialFunction=&MF_Timestamp_InOut_01}.SmoothStep)
  )
  TextureSample() {samplerType=SAMPLERTYPE_Normal, texture=/Game/Creative/Sets/PropSets/Playgrounds/Textures/T_BL_Bot_Beacon_Grenade_N}
  TextureSample() {samplerType=SAMPLERTYPE_Masks, texture=/Game/Creative/Sets/PropSets/Playgrounds/Textures/T_BL_Bot_Beacon_Grenade_S}
  Time() {bOverride_Period=true, period=1}
 ]
}
